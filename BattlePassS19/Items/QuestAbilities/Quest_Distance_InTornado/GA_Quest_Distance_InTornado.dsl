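; Season 19 quest ability that tracks meters the owning player travels, apparently inside a tornado, and reports them as quest stat events.
; A randomized 1–2 second looping timer adds floored horizontal distance, skipping movement while restricted tags are present.
; The total is sent and reset whenever it exceeds a threshold.
; Tracked-tag removal or death flushes the remainder and ends the ability; quest completion ends it without flushing.
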(event Ability|EventActivateAbilityFromEvent
  (params (EventData :type Gameplay%20Event%20Data%20Structure%20%28by%20ref%29))
  (bind _owningplayerpawn (Variables|GAVars|Data|GetOwningPlayerPawn))
  (bind _returnvalue (Transformation|GetActorLocation _owningplayerpawn))
  (|CallParentFunction EventData)
  (Development|PrintString "@@@ ActivateAbilityFromEvent [GA_Quest_Distance_Tornadol]: CALLED!" false (Variables|Debug|GetbDebugAbility))
  ;[0.1354 1 1] Initialize Tracking Parameters
  (Utilities|IsValid _owningplayerpawn
    (:Is%20Valid
      (Variables|Default|SetCurrentLocation _returnvalue)
      (Variables|Default|SetPreviousLocation _returnvalue)
      ;[0.5326 0.1448 1] Initialize Event Processing
      (EventDispatchers|BindEventtoOnDied (Variables|GAVars|Data|GetOwningPlayerPawn) (EventDispatchers|CreateEvent :!event Custom|OnDied))
      (bind _output_get (Variables|Default|SetUpdateInterval (Math|Random|RandomFloatinRange 1.0 2.0)))
      (bind _returnvalue_2 (Utilities|Time|SetTimerbyEvent (delegate Custom|UpdateDistance) _output_get true))
      (Variables|Default|SetUpdateTimer _returnvalue_2)
      ;[1 0.1051 0.2855] Cleanup
      (bind _asynctaskproxy (Ability|Tasks|WaitGameplayTagRemove (Variables|Default|GetTrackedAbilityTag) 0 true)
        (:then
          (Variables|Default|SetRemoveTagTask _asynctaskproxy))
        (:Removed
          (CallFunction|ForceStatUpdate)
          (Utilities|Time|ClearandInvalidateTimerbyHandle (Variables|Default|GetUpdateTimer))
          (Ability|EndAbility))))))

(event Custom|OnDied
  (params (DamagedActor :type Actor%20Object%20Reference) (Damage :type Float%20%28single-precision%29) (InstigatedBy :type Controller%20Object%20Reference) (DamageCauser :type Actor%20Object%20Reference) (HitLocation :type Vector) (FHitComponent :type Primitive%20Component%20Object%20Reference) (BoneName :type Name) (Momentum :type Vector))
  ;[1 0.1051 0.2855] Cleanup
  (GameplayTasks|EndTask (Variables|Default|GetRemoveTagTask))
  (CallFunction|ForceStatUpdate)
  (Utilities|Time|ClearandInvalidateTimerbyHandle (Variables|Default|GetUpdateTimer))
  (Ability|EndAbility))

(event Custom|ForceStatUpdate
  (Development|PrintString "@@@@@@ ActivateAbilityFromEvent [GA_Quest_Distance_Tornadol]: FORCING Stat Update..." false (Variables|Debug|GetbDebugAbility))
  (Quest|SendComplexCustomStatEvent (Variables|GAVars|Data|GetOwningPlayerQuestManager) 0 (Variables|Default|GetStatEventSourceTags) "(GameplayTags=)" (Variables|Default|GetMetersTravelled))
  (Development|PrintString (Utilities|String|Append "@@@@@@ ActivateAbilityFromEvent [GA_Quest_Distance_Tornadol]: Sending " (Utilities|String|ToString(Integer) (Variables|Default|GetMetersTravelled)) " meters travelled Stat Event!") false (Variables|Debug|GetbDebugAbility))
  (Variables|Default|SetMetersTravelled 0))

(event Custom|UpdateDistance
  (bind _owningplayerpawn (Variables|GAVars|Data|GetOwningPlayerPawn))
  (bind _returnvalue (Transformation|GetActorLocation _owningplayerpawn))
  (bind (_currentlocation_x _currentlocation_y) (Math|Vector|BreakVector (Variables|GetCurrentLocation_X)))
  (bind (_previouslocation_x _previouslocation_y) (Math|Vector|BreakVector (Variables|GetPreviousLocation_X)))
  (bind _owningplayerpawn_2 (Variables|GAVars|Data|GetOwningPlayerPawn))
  ;[1 0.4308 0.1465] Update Distance Travelled
  (Utilities|IsValid _owningplayerpawn_2
    (:Is%20Valid
      (bind _returnvalue_2 (MissionHelpers|HasTags (GameplayTags|GetOwnedGameplayTags _owningplayerpawn_2 :!class /Script/GameplayTags.GameplayTagAssetInterface) (Variables|Default|GetRestrictedTags) false false))
      (if _returnvalue_2
        (bind _output_get (Variables|Default|SetCurrentLocation _returnvalue))
        (Variables|Default|SetPreviousLocation _output_get)
        (else
          (Variables|Default|SetCurrentLocation _returnvalue)
          (Variables|Default|SetMetersTravelled (+ (Math|Float|Floor (/ (Math|Vector|VectorLength (Math|Vector|vector-vector (Math|Vector|MakeVector _currentlocation_x _currentlocation_y) (Math|Vector|MakeVector _previouslocation_x _previouslocation_y))) 1e+02)) (Variables|Default|GetMetersTravelled)))
          (Variables|Default|SetPreviousLocation (Variables|Default|GetCurrentLocation))
          (if (> (Variables|Default|GetMetersTravelled) (Variables|Default|GetDistanceThreshold))
            (Quest|SendComplexCustomStatEvent (Variables|GAVars|Data|GetOwningPlayerQuestManager) 0 (Variables|Default|GetStatEventSourceTags) "(GameplayTags=)" (Variables|Default|GetMetersTravelled))
            (Development|PrintString (Utilities|String|Append "@@@@@@ ActivateAbilityFromEvent [GA_Quest_Distance_Tornadol]: Sending " (Utilities|String|ToString(Integer) (Variables|Default|GetMetersTravelled)) " meters travelled Stat Event!") false (Variables|Debug|GetbDebugAbility))
            (Variables|Default|SetMetersTravelled 0)))))))

(event EventBP_OnQuestCompleted
  (params (QuestItem :type Fort%20Quest%20Item%20Object%20Reference))
  (Development|PrintString "@@@@@@@@@ ActivateAbilityFromEvent [GA_Quest_Distance_Tornadol]: Quest completed - ENDING!" false (Variables|Debug|GetbDebugAbility))
  (Utilities|Time|ClearandInvalidateTimerbyHandle (Variables|Default|GetUpdateTimer))
  (Ability|EndAbility))
